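# Unlit background material for the Impostor Baker preview scene, built around the engine's Epic Quad panorama HDRI from the asset viewer.
# It normalizes absolute world position into a direction vector and uses it as the texture lookup coordinate.
# The lookup is therefore relative to the world origin rather than to mesh UVs or the camera.
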Material /ImpostorBaker/Materials/PreviewScene/HDRI_Background
graph {
 properties {pixelDepthOffsetMode=PDOM_Legacy, shadingModel=MSM_Unlit, usedShadingModels=Unlit}
 nodes [TextureSample(UVs=Normalize(VectorInput=WorldPosition().XYZ)) {texture=/Engine/EditorMaterials/AssetViewer/EpicQuadPanorama_CC+EV1.EpicQuadPanorama_CC+EV1}]
}
